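# Example material demonstrating texture-based morph target animation through World Position Offset, using the MS_VertexAnimationTools_MorphTargets function.
# A LinearSine of Time drives the morph between two targets stored in the DemoBox texture, with DemoBox_Normals supplying the morph normals.
# Feature level switches pass the resulting offset and tangent-space vertex normal on higher feature levels and substitute zero on ES2 and ES3_1.
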Material /Engine/Functions/Engine_MaterialFunctions02/ExampleContent/TextureBasedWPO/TextureBasedMorphDemo
graph {
 properties {numCustomizedUVs=4, pixelDepthOffsetMode=PDOM_Legacy, usedShadingModels=Default Lit}
 nodes [
  Constant() {r=0.5}
  Constant()
  n1 = TextureObject() {samplerType=SAMPLERTYPE_LinearColor, texture=&/DemoBox_Normals}
  n2 = MaterialFunctionCall!(
   Morph Animation=MaterialFunctionCall(Value=Time()) {materialFunction=&LinearSine}.Linear Sine
   Morph Normal=@n1
   Morph Texture=TextureObject() {samplerType=SAMPLERTYPE_LinearColor, texture=&/DemoBox}
   Number of Morph Targets=Constant() {r=2}
   0-1 Animation Value?
   Morph Target UV
  ) {materialFunction=/Engine/Functions/Engine_MaterialFunctions02/MAXScripts/MS_VertexAnimationTools_MorphTargets}
  n3 = Constant()
  FeatureLevelSwitch(
   Default=@n2."Pixel Shader Tangent Vertex Normal (See tooltip)"
   ES2=@n3
   ES3_1=@n3
   SM4_REMOVED=@n2."Pixel Shader Tangent Vertex Normal (See tooltip)"
  )
  FeatureLevelSwitch(Default=@n2.World Position Offset, ES2=@n3, ES3_1=@n3, SM4_REMOVED=@n2.World Position Offset)
 ]
}
